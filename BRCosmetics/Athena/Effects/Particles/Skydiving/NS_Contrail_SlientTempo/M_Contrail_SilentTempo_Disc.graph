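# Unlit, two-sided translucent Niagara mesh-particle material for the glowing numeral disc in the Silent Tempo skydiving contrail.
# Numeral texture channels, tinted blue and spun in opposite directions, combine with curve-atlas radial gradients and a ParticleAge-driven angular sweep for emissive.
# Dynamic parameters drive spin speed and vertex-normal offsets, blending inside and outside radius by vertex color red.
# Particle alpha optionally takes Perlin-noise radial erosion through UseErode.
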Material /BRCosmetics/Athena/Effects/Particles/Skydiving/NS_Contrail_SlientTempo/M_Contrail_SilentTempo_Disc
graph {
 properties {
  bUsedWithNiagaraMeshParticles=true
  blendMode=BLEND_Translucent
  pixelDepthOffsetMode=PDOM_Legacy
  shadingModel=MSM_Unlit
  twoSided=true
  usedShadingModels=Unlit
 }
 nodes [
  n1 = MaterialFunctionCall(Float2=MaterialFunctionCall() {materialFunction=&VectorToRadialValue}.Radial Coordinates) {materialFunction=&BreakOutFloat2Components}
  n2 = DynamicParameter() {paramNames [RotateSpeed, InsideRadiusUpdate, ParticleAge, OutsideRadiusUpdate]}
  n3 = Add(A=OneMinus(None=@n1.R), B=LinearInterpolate(Alpha=Frac(None=@n2.ParticleAge)) {constA=-1})
  n4 = ConstantBiasScale(None=Cosine(None=@n1.R))
  n5 = Add(A=SquareRoot(None=OneMinus(None=@n4))) {constB=0.3}
  n6 = Multiply(A=Multiply(A=@n3, B=Saturate(None=@n5)), B=@n1.G)
  n7 = CurveAtlasRowParameter(CurveTime=@n3) {
   atlas=/Game/Athena/LinearColorCurves/CA_Contrail_SilentTempo
   curve=/Game/Athena/LinearColorCurves/Curve_Contrail_SilentTempo_03
   defaultValue=2
   parameterName=CurveB
  }
  n8 = TextureCoordinate() {coordinateIndex=1}
  n9 = Add(A=Time(), B=DynamicParameter() {paramNames [RotateSpeed, InsideRadiusUpdate, ParticleAge, OutsideRadiusUpdate]}.RotateSpeed)
  n10 = Rotator(
   Coordinate=MaterialFunctionCall(UVs=@n8, Texture Scale=Constant() {r=0.91}) {materialFunction=&ScaleUVsByCenter}.UVs
   Time=@n9
  )
  n11 = TextureSample(UVs=@n10) {
   samplerSource=SSM_Clamp_WorldGroupSettings
   texture=&/T_SilentTempo_Nurmerals
  }
  n12 = Constant3Vector() {constant=<rgba 0.1772 0.445 0.5 1>}
  n13 = TextureSample(UVs=@n10) {
   samplerSource=SSM_Clamp_WorldGroupSettings
   texture=&T_SilentTempo_Nurmerals
  }
  n14 = Constant3Vector() {constant=<rgba 0.0228 0.4154 1 1>}
  n15 = TextureCoordinate() {coordinateIndex=1}
  n16 = Constant() {r=1}
  n17 = TextureSample(UVs=Rotator(
    Coordinate=MaterialFunctionCall(UVs=@n15, Texture Scale=@n16) {materialFunction=&ScaleUVsByCenter}.UVs
    Time=@n9
   ) {speed=-0.25}) {
   samplerSource=SSM_Clamp_WorldGroupSettings
   texture=&T_SilentTempo_Nurmerals
  }
  n18 = Add(
   A=Add(A=Multiply(A=@n11.R, B=@n12), B=Multiply(A=@n13.B, B=@n14))
   B=Multiply(A=@n17.G, B=Constant3Vector() {constant=<rgba 0.0101 0.1471 0.25 1>})
  )
  n19 = Multiply(A=Multiply(A=Multiply(A=@n6, B=@n7.R), B=@n18), B=Constant() {r=20})
  n20 = CurveAtlasRowParameter(CurveTime=MaterialFunctionCall(Float2=MaterialFunctionCall() {materialFunction=&VectorToRadialValue}.Radial Coordinates) {materialFunction=&BreakOutFloat2Components}.G) {
   atlas=&CA_Contrail_SilentTempo
   curve=/Game/Athena/LinearColorCurves/Curve_Contrail_SilentTempo_01
   parameterName=CurveA
  }
  n21 = Add(A=@n20, B=@n18)
  n22 = ParticleColor()
  MaterialFunctionCall(Emissive=Multiply(A=Add(A=@n19, B=@n21), B=@n22.RGB)) {materialFunction=&MF_FX_EmissiveAdjustments}
  MaterialFunctionCall() {materialFunction=&DebugTimeSine}
  n23 = Constant() {r=1}
  n24 = Multiply(A=@n23)
  n25 = DynamicParameter() {paramNames [RotateSpeed, InsideRadiusUpdate, ParticleAge, OutsideRadiusUpdate]}
  n26 = Multiply(A=ScalarParameter() {defaultValue=10, parameterName=OutSideRaidus}, B=@n25.OutsideRadiusUpdate)
  n27 = Multiply(A=ScalarParameter() {defaultValue=25, parameterName=InsideRadius}, B=@n25.InsideRadiusUpdate)
  Multiply(
   A=VertexNormalWS()
   B=LinearInterpolate(A=Multiply(A=@n24, B=@n26) {constB=25}, B=Multiply(A=@n23, B=@n27) {constB=15}, Alpha=VertexColor().R)
  )
  n28 = TextureSample(UVs=TextureCoordinate()) {
   samplerSource=SSM_Wrap_WorldGroupSettings
   texture=&T_SilentTempo_Nurmerals
  }
  Multiply(A=@n28.R)
  n29 = TextureCoordinate()
  n30 = Constant() {r=0.5}
  n31 = LinearInterpolate(Alpha=DynamicParameter() {parameterIndex=1, defaultValue=<rgba 0 1 1 1>, paramNames [Erode,Param2,Param3,Param4]}.Erode) {constA=-2}
  n32 = TextureSample(UVs=Multiply(
    A=MaterialFunctionCall() {materialFunction=&VectorToRadialValue}.Radial Coordinates
    B=Constant2Vector() {g=0.2, r=2}
   )) {samplerType=SAMPLERTYPE_LinearGrayscale, texture=&T_SimplePerlin_Greyscale}
  n33 = CurveAtlasRowParameter(CurveTime=Subtract(A=SphereMask(A=@n29, B=@n30) {attenuationRadius=0.5, hardnessPercent=0}, B=Add(A=@n31, B=@n32.R))) {
   atlas=&CA_Contrail_SilentTempo
   curve=&Curve_Contrail_SilentTempo_03
   defaultValue=2
   parameterName=CurveC
  }
  Multiply(A=@n22.A, B=LinearInterpolate(B=@n33.A, Alpha=ScalarParameter() {parameterName=UseErode}) {constA=1})
  Multiply(A=@n7.R)
  Multiply(A=MaterialFunctionCall() {materialFunction=&DebugTimeSine}.Result) {constB=-1}
  Multiply(A=@n8) {constB=-1}
  OneMinus(None=@n8)
  TextureCoordinate() {coordinateIndex=1}
 ]
}
