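# Emissive material for the numeral meshes of the SilentTempo skydiving contrail, used on Niagara mesh particles.
# Vertex color red and green channels mask two color parameters, ColorA (default cyan) and ColorB (default white), which are summed and passed through MF_FX_EmissiveAdjustments to emissive.
# It uses material attributes with the Cloth shading model, near-full roughness and a tiny fixed-group parameter feeding the Y and Z of subsurface color.
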Material /BRCosmetics/Athena/Effects/Particles/Skydiving/NS_Contrail_SlientTempo/M_Contrail_SilentTempo_Numeral
graph {
 properties {
  bUseMaterialAttributes=true
  bUsedWithNiagaraMeshParticles=true
  pixelDepthOffsetMode=PDOM_Legacy
  previewMesh=&/SM_Contrail_SilentTempo_Numeral_I
  shadingModel=MSM_Cloth
  usedShadingModels=Cloth
 }
 nodes [
  n1 = VertexColor()
  n2 = Multiply(A=@n1.R, B=VectorParameter() {parameterName=ColorA, defaultValue=<rgba 0.1107 0.7992 1 1>}.RGB)
  n3 = Multiply(A=@n1.G, B=VectorParameter() {parameterName=ColorB, defaultValue=<rgba 1 1 1 1>}.RGB)
  n4 = ScalarParameter() {defaultValue=0.003, group=DontModify, parameterName=Subsurface Mask Value, sliderMax=1}
  MakeMaterialAttributes!(
   BaseColor
   Metallic
   Specular
   Roughness=Constant() {r=0.969}
   Anisotropy
   EmissiveColor=MaterialFunctionCall(Emissive=Add(A=@n2, B=@n3)) {materialFunction=&MF_FX_EmissiveAdjustments}.Result
   Opacity
   OpacityMask
   Normal
   Tangent
   WorldPositionOffset
   SubsurfaceColor=MaterialFunctionCall(Y=@n4, Z=@n4) {materialFunction=&MakeFloat3}.Result
   ClearCoat
   ClearCoatRoughness
   AmbientOcclusion
   Refraction
   CustomizedUVs_0
   CustomizedUVs_1
   CustomizedUVs_2
   CustomizedUVs_3
   CustomizedUVs_4
   CustomizedUVs_5
   CustomizedUVs_6
   CustomizedUVs_7
   PixelDepthOffset
   ShadingModel
   Displacement
  )
 ]
}
