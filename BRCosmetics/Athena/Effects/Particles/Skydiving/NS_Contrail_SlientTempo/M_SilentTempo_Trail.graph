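# Unlit translucent Niagara ribbon material for the SilentTempo skydiving contrail.
# A panning, per-particle-offset smoke texture is colored through a curve atlas and hue shift, with a curve along ribbon U modulating emissive brightness.
# Dynamic parameters offset the scroll time and blend its driver between world height and time, also feeding a sine/cosine offset vector scaled by Amplitude.
# An opacity-style mask multiplies smoke by fades at both ribbon ends.
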Material /BRCosmetics/Athena/Effects/Particles/Skydiving/NS_Contrail_SlientTempo/M_SilentTempo_Trail
graph {
 properties {bUsedWithNiagaraRibbons=true, blendMode=BLEND_Translucent, pixelDepthOffsetMode=PDOM_Legacy, shadingModel=MSM_Unlit, usedShadingModels=Unlit}
 nodes [
  n1 = MaterialFunctionCall(Float2=TextureCoordinate() {uTiling=0.2}) {materialFunction=&BreakOutFloat2Components}
  n2 = Add(A=@n1.R, B=ParticleRandom()) {constB=1.2}
  n3 = DynamicParameter()
  n4 = Add(A=Time(), B=@n3.R)
  n5 = Multiply(A=Constant() {r=0.4})
  n6 = ComponentMask(None=WorldPosition().XYZ) {b=true}
  n7 = Multiply(A=Time()) {constB=-1800}
  n8 = Constant() {r=300}
  n9 = Multiply(A=@n5) {constB=-1}
  n10 = Add(
   A=Add(A=@n5, B=Multiply(A=@n4, B=@n5) {constB=4.2})
   B=Divide(A=LinearInterpolate(A=@n6, B=@n7, Alpha=@n3.B), B=Divide(A=@n8, B=@n9)) {constB=1618}
  )
  n11 = Divide(A=@n10)
  n12 = TextureSample(UVs=Panner(Coordinate=AppendVector(A=@n2, B=@n1.G), Time=@n11) {speedX=-0.2}) {texture=/Game/Effects/Fort_Effects/Textures/Smoke/T_Smoke_Trail_Soft_2}
  n13 = CurveAtlasRowParameter(CurveTime=@n12.R) {
   atlas=/Game/Athena/LinearColorCurves/CA_Contrail_SilentTempo
   curve=/Game/Athena/LinearColorCurves/Curve_Contrail_SilentTempo_04
   defaultValue=3
   parameterName=Param_2
  }
  n14 = MaterialFunctionCall(Hue Shift Percentage=Constant() {r=0.05}, Texture=@n13) {materialFunction=&HueShift}
  n15 = MaterialFunctionCall(Float2=TextureCoordinate()) {materialFunction=&BreakOutFloat2Components}
  n16 = CurveAtlasRowParameter(CurveTime=@n15.R) {
   atlas=&CA_Contrail_SilentTempo
   curve=&Curve_Contrail_SilentTempo_04
   defaultValue=3
   parameterName=Param_1
  }
  MaterialFunctionCall(Emissive=Multiply(A=Multiply(A=@n12.R, B=@n14.Result), B=LinearInterpolate(Alpha=@n16.A) {constA=0.5, constB=2})) {materialFunction=&MF_FX_EmissiveAdjustments}
  Multiply(A=@n11, B=LinearInterpolate(Alpha=@n3.G) {constA=1, constB=-1})
  n17 = Multiply(A=TextureCoordinate(), B=Constant2Vector() {g=5, r=1})
  n18 = Frac(None=ComponentMask(None=Panner(Coordinate=@n17, Time=@n10) {speedY=0.2}) {g=true})
  Multiply(A=@n18)
  n19 = ComponentMask(None=@n17) {r=true}
  n20 = Sine(None=@n18)
  n21 = Cosine(None=@n18)
  n22 = Constant()
  Multiply(
   A=Saturate(None=Multiply(A=@n19) {constB=2})
   B=Multiply(
    A=MaterialFunctionCall(R=@n20, G=@n21, B=@n22) {materialFunction=&AppendMany}.RGB
    B=ScalarParameter() {defaultValue=28, parameterName=Amplitude}
   ) {constB=15}
  )
  n23 = OneMinus(None=@n15.R)
  n24 = Saturate(None=Multiply(A=@n15.R) {constB=15})
  Multiply(A=Multiply(A=@n12.R, B=Multiply(A=@n23, B=@n24)), B=ScalarParameter() {defaultValue=1.5, parameterName=Param_3})
 ]
}
